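# Ambient sound cue for the ticking ping of a cooling pickup truck engine, used as a subtle spot effect in city biomes.
# A quality switch feeds two parallel chains, one for the first two levels and one for the last two, each referencing its own set of identically named waves.
# Each chain loops a modulated random pick from three ping variations.
# Subtle spot attenuation and a concurrency override limit its reach and instance count.
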SoundCue /Game/Sounds/Artemis/Biomes/Cities/City_Spots/Spot_Engine_Cooling_Ping_Truck_Loop_Cue
root=QualityLevel_0
nodes [
 {id=Looping_0, children [Modulator_0]}
 {id=Looping_6, children [Modulator_3]}
 {id=Modulator_0, children [Random_1], properties {volumeMin=1}}
 {id=Modulator_3, children [Random_0], properties {volumeMin=1}}
 {id=QualityLevel_0, children [Looping_6,Looping_6,Looping_0,Looping_0]}
 {id=Random_0, children [WavePlayer_0,WavePlayer_1,WavePlayer_2], properties {weights [1,1,1]}}
 {id=Random_1, children [WavePlayer_3,WavePlayer_4,WavePlayer_5], properties {weights [1,1,1]}}
 {id=WavePlayer_0, properties {soundWaveAssetPtr=&/Spot_Engine_Cooling_Ping_PickUp_Truck_Loop_01}}
 {id=WavePlayer_1, properties {soundWaveAssetPtr=&/Spot_Engine_Cooling_Ping_PickUp_Truck_Loop_03}}
 {id=WavePlayer_2, properties {soundWaveAssetPtr=&/Spot_Engine_Cooling_Ping_PickUp_Truck_Loop_02}}
 {id=WavePlayer_3, properties {soundWaveAssetPtr=&Spot_Engine_Cooling_Ping_PickUp_Truck_Loop_01}}
 {id=WavePlayer_4, properties {soundWaveAssetPtr=&Spot_Engine_Cooling_Ping_PickUp_Truck_Loop_03}}
 {id=WavePlayer_5, properties {soundWaveAssetPtr=&Spot_Engine_Cooling_Ping_PickUp_Truck_Loop_02}}
]
properties {
 attenuationSettings=/Game/Sounds/Attenuation/Fort_TileAmbient_SpotFX_Subtle_Att
 bEnableBusSends=false
 bOverrideConcurrency=true
 maxDistance=1150
 soundClassObject=&Fort_Ambient_OneShots
 volumeMultiplier=0.752
 concurrencyOverrides {maxCount=2}
 preEffectBusSends [{}]
}
